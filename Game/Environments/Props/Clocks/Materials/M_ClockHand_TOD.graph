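# Clock hand material that rotates a prop's hand from the TimeOfDay_0-1 value in FortniteMaterialParameters, with Nanite usage enabled.
# Static switches select a smooth hour hand at two turns per day, a minute hand at twelve times that rate, or a third hand at sixty times.
# The faster hands use Floor to tick in 60 steps per turn.
# RotateAboutAxis and FixRotateAboutAxisNormals share that angle, a local-space axis and a pivot offset parameter.
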Material /Game/Environments/Props/Clocks/Materials/M_ClockHand_TOD
graph {
 properties {bUsedWithNanite=true, pixelDepthOffsetMode=PDOM_Legacy, usedShadingModels=Default Lit}
 nodes [
  MaterialFunctionCall() {materialFunction=&ObjectPivotPoint}
  n1 = Transform(None=VectorParameter() {parameterName=RotationAxis, defaultValue=<rgba 0 1 0 1>}.RGB) {transformSourceType=TRANSFORMSOURCE_Local}
  n2 = Multiply(A=CollectionParameter() {collection=&FortniteMaterialParameters, parameterName=TimeOfDay_0-1}) {constB=2, desc=Convert to 0 - 12}
  n3 = Constant() {r=60}
  n4 = Floor(None=Multiply(A=Multiply(A=@n2) {constB=12, desc=Minute hand has 12 Rotations per hour}, B=@n3) {constB=60})
  n5 = Constant() {r=60}
  n6 = Floor(None=Multiply(A=Multiply(A=@n2) {constB=60, desc=Minute hand has 12 Rotations per hour}, B=@n5) {constB=60})
  n7 = StaticSwitchParameter(
   True=@n2
   False=StaticSwitchParameter(True=Divide(A=@n4, B=@n3) {constB=60}, False=Divide(A=@n6, B=@n5) {constB=60}) {defaultValue=true, parameterName=IsMinuteHand}
  ) {parameterName=isHourHand}
  n8 = TransformPosition(None=VectorParameter() {parameterName=LocalPivotOffset, defaultValue=<rgba 0 0 0 1>}.RGB) {transformType=TRANSFORMPOSSOURCE_World}
  MaterialFunctionCall(Rotation Axis=@n1, Rotation Angle=@n7, Normals Pivot=@n8) {materialFunction=&FixRotateAboutAxisNormals}
  RotateAboutAxis(NormalizedRotationAxis=@n1, RotationAngle=@n7, PivotPoint=@n8, Position=WorldPosition().XYZ)
  TextureSampleParameter2D() {parameterName=DiffuseTexture, texture=/Game/Environments/Props/Clocks/Textures/Prop_BigClock_01_D}
  TextureSampleParameter2D() {
   parameterName=MaskTexture
   samplerType=SAMPLERTYPE_Masks
   texture=/Game/Environments/Props/Clocks/Textures/Prop_BigClock_01_S
   channelNames {a="", b=Roughness, g=Metallic, r=Specular}
  }
 ]
}
